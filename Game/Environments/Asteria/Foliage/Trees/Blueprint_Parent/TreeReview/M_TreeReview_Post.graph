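# Post-process material for tree review that shows the scene roughness buffer inside a world-space box and the normal scene color elsewhere.
# A BoxMask-3D on world position, driven by Center_Roughness and Bounds_Roughness parameters, blends post-process input with roughness.
# A feature level switch passes plain scene color on ES3_1.
# A world X comparison against MinPos_Roughness is left incomplete and feeds nothing.
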Material /Game/Environments/Asteria/Foliage/Trees/Blueprint_Parent/TreeReview/M_TreeReview_Post
graph {
 properties {materialDomain=MD_PostProcess, pixelDepthOffsetMode=PDOM_Legacy, usedShadingModels=Default Lit}
 nodes [
  n1 = SceneTexture() {sceneTextureId=PPI_PostProcessInput0}
  n2 = SceneTexture() {sceneTextureId=PPI_Roughness}
  n3 = WorldPosition()
  n4 = MaterialFunctionCall!(
   A=@n3.XYZ
   B=VectorParameter() {parameterName=Center_Roughness, defaultValue=<rgba 0 0 0 1>}.RGB
   Bounds=VectorParameter() {parameterName=Bounds_Roughness, defaultValue=<rgba 0 0 0 1>}.RGB
   Edge Falloff
  ) {materialFunction=/Engine/Functions/Engine_MaterialFunctions01/Gradient/BoxMask-3D}
  FeatureLevelSwitch(
   Default=ComponentMask(None=LinearInterpolate(A=@n1.Color, B=@n2.Color, Alpha=@n4.Result)) {b=true, g=true, r=true}
   ES3_1=SceneTexture() {sceneTextureId=PPI_PostProcessInput0}.Color
  )
  If(
   A=MaterialFunctionCall(Float3=@n3.XYZ) {materialFunction=&BreakOutFloat3Components}.R
   B=ScalarParameter() {parameterName=MinPos_Roughness}
  )
 ]
}
