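# Editor utility tree review actor that lays out a tree Blueprint for visual and budget inspection.
# Its construction script builds rows comparing the tree, non-Nanite copies, forced lowest LOD and shadow proxy meshes, with billboards for missing proxies and a post-process review material.
# Embedded Python reads mesh vertex, collision and distance field data, which helper functions turn into colour-coded text, some against size-derived budgets.
# A custom event toggles Virtual Shadow Maps.
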EditorUtilityBlueprint /Game/Environments/Asteria/Foliage/Trees/Blueprint_Parent/TreeReview/EUBP_Test
graphs [
 AddTextCheckBool {nodes [
   n1 = AddTextCheckBool!() -> (then, Value:bool=~, Description:String, InvertColor:bool=~) {is=title K2Node_FunctionEntry}
   > Add(
    TargetArray=Get CheckList() -> (CheckList:Array of Strings)
    NewItem=Append(A=Append(A=@n1.Description, B=": "), B="To String (Boolean)"(InBool=@n1))
   ) {is=title K2Node_CallArrayFunction}
   n2 = LinearColor_Green!() -> (ReturnValue:Linear Color Structure)
   n3 = LinearColor_Red!() -> (ReturnValue:Linear Color Structure)
   n4 = NOT Boolean(A=@n1)
   > Add(
    TargetArray:Array of Color Structures=Get CheckListColor() -> (CheckListColor:Array of Color Structures)
    NewItem:"Color Structure (by ref)"=Select(
     Option 0:"Color Structure (by ref)"="To Color (LinearColor)"!(InLinearColor:Linear Color Structure=@n3, InUseSRGB:bool=true) -> (ReturnValue:Color Structure)
     Option 1:"Color Structure (by ref)"="To Color (LinearColor)"(InLinearColor=@n2, InUseSRGB=true)
     Index=Select(Option 0:bool=@n1, Option 1:bool=@n4, Index=@n1.InvertColor) -> (ReturnValue:bool)
    ) -> (ReturnValue:"Color Structure (by ref)")
   )
  ]}
 AddTextCheckFloat {nodes [
   n1 = AddTextCheckFloat!() -> (then, Value:double=~, Description:String, ThresholdGreen:double=~, ThresholdRed:double=~, UseColorGradient:bool=~) {is=title K2Node_FunctionEntry}
   > Add(
    TargetArray:Array of Strings=Get CheckList()
    NewItem:"String (by ref)"=Append(A=Append(A=@n1.Description, B=": "), B="To String (Float)"(InDouble=@n1))
   )
   n2 = "Max (Float)"(A=float - float(A=@n1.ThresholdRed, B=@n1.ThresholdGreen), B=0.001)
   n3 = float - float(A=@n1, B=@n1.ThresholdGreen)
   n4 = "Lerp (LinearColor)"(A=LinearColor_Green(), B=LinearColor_Red(), Alpha="Clamp (Float)"(Value=float / float(A=@n3, B=@n2), Min=0, Max=1))
   n5 = "To Color (LinearColor)"(InLinearColor=LinearColor_Red(), InUseSRGB=true)
   n6 = "float < float"(A=@n1, B=@n1.ThresholdRed)
   n7 = Select(
    Option 0:"Color Structure (by ref)"="To Color (LinearColor)"(InLinearColor=LinearColor_Yellow!() -> (ReturnValue:Linear Color Structure), InUseSRGB=true)
    Option 1:"Color Structure (by ref)"="To Color (LinearColor)"(InLinearColor=LinearColor_Green(), InUseSRGB=true)
    Index="float > float"(A=@n1, B=@n1.ThresholdGreen)
   ) -> (ReturnValue:"Color Structure (by ref)")
   > Add(
    TargetArray:Array of Color Structures=Get CheckListColor()
    NewItem:"Color Structure (by ref)"=Select(
     Option 0=Select(Option 0=@n7, Option 1=@n5, Index=@n6)
     Option 1="To Color (LinearColor)"(InLinearColor=@n4, InUseSRGB=true)
     Index=@n1.UseColorGradient
    )
   )
  ]}
 AddTextCheckInt {nodes [
   n1 = AddTextCheckInt!() -> (then, Value:int, Description:String, ThresholdGreen:int, ThresholdRed:int, UseColorGradient:bool=~, ScaleByBudgetSize:bool=~) {is=title K2Node_FunctionEntry}
   > Add(
    TargetArray:Array of Strings=Get CheckList()
    NewItem:"String (by ref)"=Append(A=Append(A=@n1.Description, B=": "), B="To String (Integer)"(InInt=@n1))
   )
   n2 = SelectFloat(A=Get BudgetMultiplier() -> (BudgetMultiplier:double), B=1, bPickA=@n1.ScaleByBudgetSize)
   n3 = float * float(A:int=@n1.ThresholdGreen, B=@n2) -> (ReturnValue:int) {is=title K2Node_PromotableOperator}
   n4 = float * float(A=@n1.ThresholdRed, B=@n2)
   n5 = "To Float (Integer)"(InInt="Max (Integer)"(A=int - int(A=@n4, B=@n3), B=1))
   n6 = "To Float (Integer)"(InInt=int - int(A=@n1, B=@n3))
   n7 = LinearColor_Green()
   n8 = LinearColor_Red()
   n9 = "Lerp (LinearColor)"(A=@n7, B=@n8, Alpha="Clamp (Float)"(Value=float / float(A=@n6, B=@n5), Min=0, Max=1))
   n10 = "To Color (LinearColor)"(InLinearColor=@n8, InUseSRGB=true)
   n11 = "integer < integer"(A=@n1, B=@n4)
   n12 = Select(
    Option 0="To Color (LinearColor)"(InLinearColor=LinearColor_Yellow(), InUseSRGB=true)
    Option 1="To Color (LinearColor)"(InLinearColor=@n7, InUseSRGB=true)
    Index="integer > integer"(A=@n1, B=@n3)
   )
   > Add(
    TargetArray:Array of Color Structures=Get CheckListColor()
    NewItem:"Color Structure (by ref)"=Select(
     Option 0=Select(Option 0=@n12, Option 1=@n10, Index=@n11)
     Option 1="To Color (LinearColor)"(InLinearColor=@n9, InUseSRGB=true)
     Index=@n1.UseColorGradient
    )
   )
  ]}
 EventGraph {nodes [
   ToggleShadowVSM() {is=title K2Node_CustomEvent}
   > n1 = Branch(Condition="Equal (Float)"(A=GetNumericCVar!(CVar:String=r.Shadow.Virtual.Enable) -> (OutValue:float, ReturnValue:bool)))
   > ExecuteConsoleCommand(Command=r.Shadow.Virtual.Enable 1)
   ExecuteConsoleCommand(execute=@n1.else, Command=r.Shadow.Virtual.Enable 0)
   Event BeginPlay()
   Event ActorBeginOverlap()
   Event Tick!() -> (OutputDelegate, then, DeltaSeconds:float)
  ]}
 UserConstructionScript {nodes [
   Construction Script()
   > n1 = Sequence!(execute) -> (then_0, then_1, then_2:Exec)
   > n2 = For Loop(LastIndex=Get NumRows() -> (NumRows:int=0))
   > Set Iteration(Iteration:int=@n2) -> (Output_Get:int=0)
   > Set NumComponent(NumComponent:int=1) -> (Output_Get:int=0)
   > n3 = Get!(execute:Exec) -> (then:Exec, else:Exec, Blueprint:Actor Class Reference)
   n4 = float * float(A:double=Get BoundsY() -> (BoundsY:double), B:int=Get Iteration() -> (Iteration:int=0)) -> (ReturnValue:double)
   > n5 = Add Child Actor Component!(
    execute:Exec
    self:Actor Object Reference
    bManualAttachment:bool=false
    RelativeTransform:Transform=MakeTransform(Location=MakeVector(Y=@n4), Rotation="0,90,0")
    ComponentTemplateContext:Object Reference
   ) -> (then:Exec, ReturnValue:Child Actor Component Object Reference) {is=title K2Node_AddComponent}
   > SetChildActorClass!(execute, self:Child Actor Component Object Reference=@n5, InClass:Actor Class Reference=@n3) -> (then)
   > n6 = Sequence()
   n7 = Get ChildActor!(self:Child Actor Component Object Reference=@n5) -> (ChildActor:Actor Object Reference)
   n8 = "Get (a copy)"(Array:Array of Static Mesh Component Object References=GetComponentsByTag!(self:Actor Object Reference=@n7, ComponentClass:Actor Component Class Reference=&StaticMeshComponent, Tag:Name=RenderMesh) -> (ReturnValue:Array of Static Mesh Component Object References)) -> (Output:Static Mesh Component Object Reference)
   > n9 = Set RenderMeshComponent(RenderMeshComponent:Static Mesh Component Object Reference=@n8) -> (Output_Get:Static Mesh Component Object Reference)
   > Set RenderMesh(RenderMesh:Static Mesh Object Reference=Get StaticMesh(self=@n8)) -> (Output_Get:Static Mesh Object Reference)
   n10 = GetLocalBounds!(self:Static Mesh Component Object Reference=@n9) -> (Min:Vector, Max:Vector)
   n11 = BreakVector(InVec=vector - vector(A=@n10.Max, B=@n10))
   n12 = float / float(A=float * float(A="Max (Float)"(A=@n11, B=@n11.Y), B:double=@n11.Z), B=100000)
   > Set MeshVolumeSize(MeshVolumeSize:double="To Float (Integer)"(InInt=Round(A=@n12) {is=title K2Node_CallFunction})) -> (Output_Get:double)
   n13 = BreakVector(InVec=@n10)
   n14 = BreakVector(InVec=@n10.Max)
   > Set BoundsX(BoundsX:double=float + float(A=float - float(A=@n14, B=@n13), B=512)) -> (Output_Get:double)
   > Set BoundsY(BoundsY:double=float * float(A=float - float(A=@n13.Y, B=@n14.Y), B=2)) -> (Output_Get:double)
   > Set BoundZMax(BoundZMax:double=float + float(A=@n14.Z, B=256)) -> (Output_Get:double)
   n15 = float * float(A=Get BoundsY(), B:int=Get Iteration() -> (Iteration=0))
   n16 = Get BoundZMax() -> (BoundZMax:double)
   > Add Text Render Component!(
    execute:Exec
    self:Actor Object Reference
    bManualAttachment:bool=false
    RelativeTransform:Transform=MakeTransform(Location=MakeVector(Y=@n15, Z=@n16), Rotation="0,90,0")
    ComponentTemplateContext:Object Reference
   ) -> (then:Exec, ReturnValue:Text Render Component Object Reference) {is=title K2Node_AddComponent}
   > n17 = Sequence!(execute) -> (then_0, then_1)
   n18 = float * float(A=Get BoundsY(), B=Get Iteration() -> (Iteration=0))
   n19 = float * float(A=Get BoundsX() -> (BoundsX:double), B=Get NumComponent() -> (NumComponent:int=0))
   > n20 = Add Child Actor Component(bManualAttachment=false, RelativeTransform=MakeTransform(Location=MakeVector(X=@n19, Y=@n18), Rotation="0, 0, 0"))
   > SetChildActorClass(self=@n20, InClass=Get Blueprint() -> (Blueprint:Actor Class Reference))
   n21 = Get ChildActor(self=@n20)
   > Set bDisallowNanite!(
    execute
    bDisallowNanite:bool=true
    self:Static Mesh Component Object Reference="Get (a copy)"(Array=GetComponentsByTag(self=@n21, ComponentClass=&StaticMeshComponent, Tag=RenderMesh))
   ) -> (then, Output_Get:bool)
   n22 = float * float(A=Get BoundsY(), B=Get Iteration() -> (Iteration=0))
   n23 = Get BoundZMax()
   > Add Text Render Component(bManualAttachment=false, RelativeTransform=MakeTransform(Location=MakeVector(X=@n19, Y=@n22, Z=@n23), Rotation="0,90,0"))
   > Increment Int(Value=Get NumComponent() -> (NumComponent=0))
   Is Valid(exec=@n17.then_1, InputObject=Get RenderMesh() -> (RenderMesh:Static Mesh Object Reference))
   n24 = float * float(A=Get BoundsY(), B=Get Iteration() -> (Iteration=0))
   n25 = float * float(A=Get BoundsX(), B=Get NumComponent() -> (NumComponent=0))
   > n26 = Add Static Mesh Component!(
    execute:Exec
    self:Actor Object Reference
    bManualAttachment:bool=false
    RelativeTransform:Transform=MakeTransform(Location=MakeVector(X=@n25, Y=@n24), Rotation="0, 0, 0")
    ComponentTemplateContext:Object Reference
   ) -> (then:Exec, ReturnValue:Static Mesh Component Object Reference) {is=title K2Node_AddComponent}
   > SetStaticMesh(self=@n26, NewMesh=Get RenderMesh())
   > Set bDisallowNanite(bDisallowNanite=true, self=@n26)
   > SetForcedLodModel!(execute, self:Static Mesh Component Object Reference=@n26, NewForcedLodModel:int=100) -> (then)
   n27 = float * float(A=Get BoundsY(), B=Get Iteration() -> (Iteration=0))
   n28 = Get BoundZMax()
   > Add Text Render Component(bManualAttachment=false, RelativeTransform=MakeTransform(Location=MakeVector(X=@n25, Y=@n27, Z=@n28), Rotation="0,90,0"))
   > Increment Int(Value=Get NumComponent() -> (NumComponent=0))
   n29 = Sequence!(execute=@n6.then_1) -> (then_0, then_1, then_2:Exec)
   n30 = Get StaticMesh(self="Get (a copy)"(Array=GetComponentsByTag(self=@n7, ComponentClass=&StaticMeshComponent, Tag=ShadowProxy)))
   > n31 = Is Valid(InputObject=@n30)
   n32 = Get Iteration() -> (Iteration=0)
   n33 = Get BoundsY()
   n34 = Get BoundsX()
   n35 = float * float(A=@n34, B=Get NumComponent() -> (NumComponent=0))
   n36 = Get BoundZMax()
   n37 = MakeTransform(
    Location=MakeVector(X=@n35, Y=float * float(A=@n33, B=@n32), Z=float / float(A=@n36, B=2))
    Rotation="0, 0, 0"
    Scale="To Vector (Float)"!(InDouble:double=float / float(A=@n34, B=100)) -> (ReturnValue:Vector)
   )
   Add Billboard Component!(
    execute:Exec=@n31.Is Not Valid
    self:Actor Object Reference
    bManualAttachment:bool=false
    RelativeTransform:Transform=@n37
    ComponentTemplateContext:Object Reference
   ) -> (then:Exec, ReturnValue:Billboard Component Object Reference) {is=title K2Node_AddComponent}
   > Increment Int(Value=Get NumComponent() -> (NumComponent=0))
   n38 = Get Iteration() -> (Iteration=0)
   n39 = Get BoundsY()
   n40 = MakeTransform(Location=MakeVector(X=@n35, Y=float * float(A=@n39, B=@n38)), Rotation="0, 0, 0")
   n41 = Add Static Mesh Component(execute=@n31.Is Valid, bManualAttachment=false, RelativeTransform=@n40)
   > SetStaticMesh(self=@n41, NewMesh=@n30)
   n42 = Get Iteration() -> (Iteration=0)
   n43 = Get BoundsY()
   n44 = MakeTransform(Location=MakeVector(X=@n35, Y=float * float(A=@n43, B=@n42), Z=Get BoundZMax()), Rotation="0,90,0")
   > n45 = Add Text Render Component(bManualAttachment=false, RelativeTransform=@n44)
   Increment Int(" "=[@n45.then, @n48.then], Value=Get NumComponent() -> (NumComponent=0))
   n46 = Is Valid(exec=@n29.then_1, InputObject=@n30)
   Add Billboard Component(execute=@n46.Is Not Valid, bManualAttachment=false, RelativeTransform=@n37)
   > Increment Int(Value=Get NumComponent() -> (NumComponent=0))
   n47 = Add Static Mesh Component(execute=@n46.Is Valid, bManualAttachment=false, RelativeTransform=@n40)
   > SetStaticMesh(self=@n47, NewMesh=@n30)
   > Set bDisallowNanite(bDisallowNanite=true, self=@n47)
   > n48 = Add Text Render Component(bManualAttachment=false, RelativeTransform=@n44)
   n49 = float * float(A=Get BoundsY(), B=Get NumRows() -> (NumRows=0))
   n50 = Get NumComponent() -> (NumComponent=0)
   n51 = Get BoundsX()
   n52 = Get BoundZMax()
   n53 = Get Box() -> (Box:Box Collision Object Reference)
   SetBoxExtent!(
    execute=@n1.then_1
    self:Box Collision Object Reference=@n53
    InBoxExtent:Vector=MakeVector(X=float * float(A=@n51, B=@n50), Y=@n49, Z=@n52)
    bUpdateOverlaps:bool=true
   ) -> (then)
   > Set Relative Location!(
    execute
    self:Scene Component Object Reference=@n53
    NewLocation:Vector=MakeVector(Y=float / float(A=@n49, B=2), Z=float / float(A=@n52, B=2))
    bSweep:bool=false
    bTeleport:bool=false
   ) -> (then, SweepHitResult:Hit Result Structure) {is=title K2Node_CallFunction}
   > n54 = CreateDynamicMaterialInstance(Parent=&/M_TreeReview_Post, OptionalName=None, CreationFlags=None)
   n55 = Make Weighted Blendable!(Weight:float=-1, Object:Object Reference=@n54) -> (WeightedBlendable:Weighted Blendable Structure) {is=title K2Node_MakeStruct}
   > n56 = Set members in Weighted Blendables!(
    execute:Exec
    StructRef:"Weighted Blendables Structure (by ref)"=Make Weighted Blendables!(Array:Array of Weighted Blendable Structures=Make Array("[0]":Weighted Blendable Structure=@n55) -> (Array:Array of Weighted Blendable Structures)) -> (WeightedBlendables:Weighted Blendables Structure) {is=title K2Node_MakeStruct}
   ) -> (then:Exec, StructOut:"Weighted Blendables Structure (by ref)") {is=title K2Node_SetFieldsInStruct}
   n57 = Get PostProcess() -> (PostProcess:Post Process Component Object Reference)
   > Set Settings!(
    execute
    Settings:Post Process Settings Structure=Make Post Process Settings!(WeightedBlendables:Weighted Blendables Structure=@n56) -> (PostProcessSettings:Post Process Settings Structure) {is=title K2Node_MakeStruct}
    self:Post Process Component Object Reference=@n57
   ) -> (then, Output_Get:Post Process Settings Structure)
   n58 = Get Settings!(self:Post Process Component Object Reference=@n57) -> (Settings:Post Process Settings Structure)
   > Set Array Elem!(
    execute
    TargetArray:Array of Weighted Blendable Structures=Break Weighted Blendables!(WeightedBlendables:"Weighted Blendables Structure (by ref)"=Break Post Process Settings!(PostProcessSettings:"Post Process Settings Structure (by ref)"=@n58) -> (WeightedBlendables:Weighted Blendables Structure) {is=title K2Node_BreakStruct}) -> (Array:Array of Weighted Blendable Structures) {is=title K2Node_BreakStruct}
    Index:int=0
    Item:"Weighted Blendable Structure (by ref)"=Make Weighted Blendable(Weight=-1, Object=@n54)
    bSizeToFit:bool=false
   ) -> (then)
   n59 = Sequence!(execute=@n1.then_2) -> (then_0, then_1)
   > n60 = GetPathNameForLoadedAsset(LoadedAsset=Get RenderMesh())
   > n61 = Execute Python Script!(
    execute
    PythonScript=Append!(
     A="import unreal\r\n"
     B
     C:String="\r\nsm = '"
     D:String=@n60
     E:String="'\r\n\r\n\r\nsmload = unreal.load_asset(sm,None,True)\r\nasset_options = unreal.GeometryScriptCopyMeshFromAssetOptions()\r\nrequested_lod = unreal.GeometryScriptMeshReadLOD()\r\ndm = unreal.DynamicMesh()\r\nunreal.GeometryScript_AssetUtils.copy_mesh_from_static_mesh(smload, dm, asset_options, requested_lod)\r\n\r\n\r\nmesh_editor_ss = unreal.get_editor_subsystem(unreal.StaticMeshEditorSubsystem)\r\nbuildsetting = mesh_editor_ss.get_lod_build_settings(smload,0)\r\nsubsys = unreal.get_editor_subsystem(unreal.StaticMeshEditorSubsystem)\r\n\r\n\r\nlod_for_collision= smload.get_editor_property(\"lod_for_collision\")\r\n \r\n\r\ndist_field_bias = smload.get_editor_property(\"distance_field_self_shadow_bias\")\r\ndist2sided = buildsetting.get_editor_property(\"generate_distance_field_as_if_two_sided\" )\r\nvertcount = subsys.get_number_verts(smload,0)\r\nnanitevertcount = dm.get_vertex_count()\r\ncollcount = subsys.get_convex_collision_count(smload)\r\nleafcol = subsys.is_section_collision_enabled(smload,lod_for_collision,1) and subsys.is_section_collision_enabled(smload,lod_for_collision,0)"
    ) -> (ReturnValue)
   ) -> (
    then
    ReturnValue
    dist_field_bias:double=~
    dist2sided:bool=~
    lod_for_collision:int
    vertcount:int
    nanitevertcount:int
    collcount:int
    leafcol:bool=~
   ) {outputs [dist_field_bias, dist2sided, lod_for_collision, vertcount, nanitevertcount, collcount, leafcol]}
   n62 = Sqrt!(A:double=Get MeshVolumeSize() -> (MeshVolumeSize:double)) -> (ReturnValue:double)
   > AddTextCheckFloat!(
    execute
    self
    Value:double=@n62
    Description:String=Tree Size
    ThresholdGreen:double=0
    ThresholdRed:double=350000
    UseColorGradient:bool=true
   ) -> (then) {is=title K2Node_CallFunction}
   n63 = float / float(A=@n62, B=10)
   > n64 = Set BudgetMultiplier(BudgetMultiplier:double=MapRangeClamped(Value="Clamp (Float)"(Value=@n63, Min=0, Max=1), InRangeA=1, InRangeB=0, OutRangeA=1, OutRangeB=Make Literal Float(Value=0))) -> (Output_Get:double)
   > AddTextCheckFloat(Value=@n64, Description=Budget Multiplier, ThresholdGreen=0, ThresholdRed=0, UseColorGradient=false)
   > AddTextCheckFloat(Value=@n61.dist_field_bias, Description=DistanceFieldSelfShadowBias, ThresholdGreen=0, ThresholdRed=0, UseColorGradient=false)
   > AddTextCheckBool!(execute, self, Value:bool=@n61.dist2sided, Description:String=TwoSidedDistanceField, InvertColor:bool=false) -> (then) {is=title K2Node_CallFunction}
   > AddTextCheckInt!(
    execute
    self
    Value:int=@n61.lod_for_collision
    Description:String=Lod for Collision
    ThresholdGreen:int=1
    ThresholdRed:int=1
    UseColorGradient:bool=false
    ScaleByBudgetSize:bool=false
   ) -> (then) {is=title K2Node_CallFunction}
   n65 = Truncate(A=float * float(A=19000, B:double=Get BudgetMultiplier()))
   > AddTextCheckInt(
    Value=@n61.vertcount
    Description=Vertex Count LOD 0
    ThresholdGreen=@n65
    ThresholdRed=int + int(A=@n65, B=1000)
    UseColorGradient=true
    ScaleByBudgetSize=false
   )
   > AddTextCheckInt(
    Value="Clamp (Integer)"(Value=@n65, Min=0, Max=12000) -> (ReturnValue=0)
    Description=Vertex Count LOD 0 TargetBudget
    ThresholdGreen=0
    ThresholdRed=0
    UseColorGradient=false
    ScaleByBudgetSize=false
   )
   n66 = Truncate(A=float * float(A=350000, B=Get BudgetMultiplier()))
   > AddTextCheckInt(
    Value=@n61.nanitevertcount
    Description=NaniteVertex Count Estimate
    ThresholdGreen=@n66
    ThresholdRed=int + int(A=@n66, B=20000)
    UseColorGradient=true
    ScaleByBudgetSize=false
   )
   > AddTextCheckInt(Value=@n66, Description=NaniteVertex Count TargetBudget, ThresholdGreen=0, ThresholdRed=0, UseColorGradient=false, ScaleByBudgetSize=false)
   > AddTextCheckInt(
    Value=@n61.collcount
    Description=Num Collision Primitives
    ThresholdGreen=1
    ThresholdRed=40
    UseColorGradient=true
    ScaleByBudgetSize=false
   )
   > AddTextCheckBool(Value=@n61.leafcol, Description="Leaves Complex Collision Maybe Enabled ", InvertColor=true)
   n67 = For Each Loop(Exec=@n59.then_1, Array=Get CheckList())
   n68 = MakeVector(X=-2000, Y=0, Z=1000)
   n69 = MakeVector(Y=0, Z=float * float(A=Make Literal Float(Value=200), B:int=@n67.Array Index))
   > n70 = Add Text Render Component(bManualAttachment=false, RelativeTransform=MakeTransform(Location=vector + vector(A=@n68, B=@n69), Rotation="0,90,0"))
   > Set Text!(execute, self:Text Render Component Object Reference=@n70, Value:"Text (by ref)"="To Text (String)"(InString=@n67)) -> (then) {is=title K2Node_CallFunction}
   > SetTextRenderColor!(
    execute
    self:Text Render Component Object Reference=@n70
    Value:Color Structure="Get (a copy)"(Array:Array of Color Structures=Get CheckListColor(), Dimension 1=@n67.Array Index) -> (Output:Color Structure)
   ) -> (then)
   Append(
    A="import unreal\r\n"
    C="\r\nsm = '"
    E="'\r\n\r\n\r\nsmload = unreal.load_asset(sm,None,True)\r\n\r\nmesh_editor_ss = unreal.get_editor_subsystem(unreal.StaticMeshEditorSubsystem)\r\n\r\n##nanite_setting = smload.get_editor_property(\"NaniteSettings\")\r\n##nanite_setting = mesh_editor_ss.get_nanite_settings(smload)\r\n##nanite_setting.set_editor_property(\"enabled\",True)\r\n##nanite_setting.set_editor_property(\"preserve_area\",True)\r\n##nanite_setting.fallback_relative_error = 0.75\r\n##nanite_setting.fallback_percent_triangles = 80.0\r\n## this applies the nanite settings\r\n\r\n\r\n##mesh_editor_ss.set_nanite_settings(smload, nanite_setting, True)\r\ndist_field_bias = smload.get_editor_property(\"distance_field_self_shadow_bias\")\r\ndist2sided = smload.get_editor_property(\"generate_distance_field_as_if_two_sided\")"
   )
   Append(
    A="import unreal\r\n"
    C="\r\nsm = '"
    E="'\r\n\r\n\r\nsmload = unreal.load_asset(sm,None,True)\r\n\r\n\r\n##nanite_setting = smload.get_editor_property(\"NaniteSettings\")\r\n##nanite_setting.fallback_relative_error = 0.75\r\n##nanite_setting.fallback_percent_triangles = 80.0\r\n## this applies the nanite settings\r\n##mesh_editor_ss = ##unreal.get_editor_subsystem(unreal.StaticMeshEditorSubsystem)\r\n\r\nlodset = mesh_editor_ss.get_lod_build_settings(smload, 0)\r\nlodset.set_editor_property(\"generate_distance_field_as_if_two_sided\", True )\r\nmesh_editor_ss.set_lod_build_settings(smload,0,lodset)\r\n##unreal.EditorStaticMeshLibrary.set_lod_build_settings(smload,0,lodset)\r\n##smload.set_editor_property(\"twosided_distance_field_generation\", ##True)\r\n##smload.set_editor_property(\"distance_field_self_shadow_bias\", ##30.0)"
   )
   float + float(A=@n49)
   float * float!(A=@n11, B:double=@n11.Y, C:double=@n11.Z) -> (ReturnValue)
  ]}
]
variables [
 Blueprint {type=Class, default=/Game/Environments/Asteria/Foliage/Trees/Oak/Blueprint/BP_Asteria_Oak_Tree_Large.BP_Asteria_Oak_Tree_Large_C}
 BoundsX {type=number}
 BoundsY {type=number}
 BoundZMax {type=number}
 BudgetMultiplier {type=number}
 CheckList {type="array<string>", default []}
 CheckListColor {type="array<Color>", default []}
 Iteration {type=integer}
 MeshVolumeSize {type=number}
 NumComponent {type=integer}
 NumRows {type=integer}
 RenderMesh {type=StaticMesh, default=null}
 RenderMeshComponent {type=StaticMeshComponent, default=null}
]
